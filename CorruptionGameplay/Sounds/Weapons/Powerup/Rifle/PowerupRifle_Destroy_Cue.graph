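# Sound cue for the destroy sound of the powerup rifle in the CorruptionGameplay weapons content.
# A single wave player plays a rifle destroy wave from the shared pistol degrade sounds, with no randomization or mixing.
# The cue uses the degraded low-ammo ranged weapon sound class and overrides concurrency so new instances are blocked at the limit rather than stopping playing ones.
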SoundCue /CorruptionGameplay/Sounds/Weapons/Powerup/Rifle/PowerupRifle_Destroy_Cue
root=WavePlayer_59
nodes [{id=WavePlayer_59, properties {soundWaveAssetPtr=/Game/Sounds/Fort_Weapons/pistol/degrades/rifle_destroy_01}}]
properties {
 bOverrideConcurrency=true
 maxDistance=2097152
 soundClassObject=/Game/Sounds/FortSoundClassesAndModes/Fort_WeaponRanged_LowAmmo_Degraded_PSM
 volumeMultiplier=1
 concurrencyOverrides {resolutionRule=PreventNew}
}
