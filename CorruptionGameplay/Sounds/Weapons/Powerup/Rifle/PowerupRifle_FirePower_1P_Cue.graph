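# Sound cue for the first-person fire-power shot of the powerup rifle in CorruptionGameplay.
# A Random node picks one of five equally weighted wave variations and feeds a Modulator that varies pitch on each play.
# The cue raises overall pitch and lowers volume, using first-person ranged-fire sound class and projectile attenuation assets.
# It joins the shared first-person gunfire concurrency set and has its debug flag on.
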SoundCue /CorruptionGameplay/Sounds/Weapons/Powerup/Rifle/PowerupRifle_FirePower_1P_Cue
root=Modulator_0
nodes [
 {id=Modulator_0, children [Random_0], properties {pitchMin=0.97}}
 {id=Random_0, children [WavePlayer_2, WavePlayer_4, WavePlayer_5, WavePlayer_9, WavePlayer_10], properties {weights [1,1,1,1,1]}}
 {id=WavePlayer_10, properties {soundWaveAssetPtr=&/PowerupRifle_FirePower_05}}
 {id=WavePlayer_2, properties {soundWaveAssetPtr=&/PowerupRifle_FirePower_01}}
 {id=WavePlayer_4, properties {soundWaveAssetPtr=&/PowerupRifle_FirePower_02}}
 {id=WavePlayer_5, properties {soundWaveAssetPtr=&/PowerupRifle_FirePower_03}}
 {id=WavePlayer_9, properties {soundWaveAssetPtr=&/PowerupRifle_FirePower_04}}
]
properties {
 attenuationSettings=&Fort_Weapon_Projectile_FirstPerson_Athena_Att
 bDebug=true
 maxDistance=2097152
 pitchMultiplier=1.1
 soundClassObject=&Fort_WeaponRanged_Fire_1P_PSM
 volumeMultiplier=0.45
 concurrencyOverrides {bLimitToOwner=true, maxCount=3, resolutionRule=StopOldest}
 concurrencySet [&Weapon_Gunfire_FirstPerson_Conc]
}
