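# Sound cue for the first-person fire sound of the powerup rifle.
# A Random node picks one of five fire wave variants at equal weight and feeds a Modulator that varies each shot.
# The cue lowers overall volume and pitch slightly, using first-person weapon attenuation, a ranged-fire sound class and gunfire concurrency.
# Owner-limited concurrency overrides that stop the oldest sound are also set, and the debug flag is on.
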SoundCue /CorruptionGameplay/Sounds/Weapons/Powerup/Rifle/PowerupRifle_Fire_1P_Cue
root=Modulator_0
nodes [
 {id=Modulator_0, children [Random_1]}
 {id=Random_1, children [WavePlayer_0,WavePlayer_1,WavePlayer_3,WavePlayer_2,WavePlayer_4], properties {weights [1,1,1,1,1]}}
 {id=WavePlayer_0, properties {soundWaveAssetPtr=&/PowerupRifle_Fire_01}}
 {id=WavePlayer_1, properties {soundWaveAssetPtr=&/PowerupRifle_Fire_02}}
 {id=WavePlayer_2, properties {soundWaveAssetPtr=&/PowerupRifle_Fire_06}}
 {id=WavePlayer_3, properties {soundWaveAssetPtr=&/PowerupRifle_Fire_03}}
 {id=WavePlayer_4, properties {soundWaveAssetPtr=&/PowerupRifle_Fire_05}}
]
properties {
 attenuationSettings=&Fort_Weapon_Projectile_FirstPerson_Athena_Att
 bDebug=true
 maxDistance=2097152
 pitchMultiplier=0.97
 soundClassObject=&Fort_WeaponRanged_Fire_1P_PSM
 volumeMultiplier=0.45
 concurrencyOverrides {bLimitToOwner=true, maxCount=3, resolutionRule=StopOldest}
 concurrencySet [&Weapon_Gunfire_FirstPerson_Conc]
}
