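# Sound cue for the end-of-reload foley of the Powerup rifle weapon.
# A mixer plays the main reload-end wave alongside a quieter layer wave, which passes through a modulator that lowers its pitch by a fixed amount.
# Output is scaled down and routed through the weapon foley sound class, with weapon reload attenuation, a concurrency override and debug enabled.
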SoundCue /CorruptionGameplay/Sounds/Weapons/Powerup/Rifle/PowerupRifle_ReloadEnd_Cue
root=Mixer_0
nodes [
 {id=Mixer_0, children [WavePlayer_0,Modulator_0], properties {inputVolume [1,0.5]}}
 {id=Modulator_0, children [WavePlayer_1], properties {pitchMax=0.9, pitchMin=0.9}}
 {id=WavePlayer_0, properties {soundWaveAssetPtr=&/PowerupRifle_ReloadEnd_02}}
 {id=WavePlayer_1, properties {soundWaveAssetPtr=&/PowerupRifle_ReloadEndLayer_02}}
]
properties {
 attenuationSettings=&Fort_Weapon_Reload_Att
 bDebug=true
 bOverrideConcurrency=true
 maxDistance=2250
 soundClassObject=&Fort_WeaponFoley
 volumeMultiplier=0.2
}
