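# Sound cue for the heavy whip sound of the powerup rifle, classed as a ranged weapon projectile effect and limited by bullet whiz-by concurrency.
# A Quality Level node sends its first two slots to a six-wave random pool and its last two to a two-wave pool.
# A root Modulator varies the result with a raised maximum pitch.
# The cue overrides attenuation to disable distance falloff and stereo spread.
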SoundCue /CorruptionGameplay/Sounds/Weapons/Powerup/Rifle/PowerupRifle_Whip_Heavy_Cue
root=Modulator_0
nodes [
 {id=Modulator_0, children [QualityLevel_0], properties {pitchMax=1.5}}
 {id=QualityLevel_0, children [Random_2,Random_2,Random_1,Random_1]}
 {id=Random_1, children [WavePlayer_11, WavePlayer_10], properties {bShouldExcludeFromBranchCulling=true, weights [1,1]}}
 {id=Random_2, children [WavePlayer_6,WavePlayer_7,WavePlayer_8,WavePlayer_9,WavePlayer_0,WavePlayer_1], properties {weights [1,1,1,1,1,1]}}
 {id=WavePlayer_0, properties {soundWaveAssetPtr=&/PowerupRifle_Whip_Heavy_06}}
 {id=WavePlayer_1, properties {soundWaveAssetPtr=&/PowerupRifle_Whip_Heavy_05}}
 {id=WavePlayer_10, properties {soundWaveAssetPtr=&/PowerupRifle_Whip_Heavy_02}}
 {id=WavePlayer_11, properties {soundWaveAssetPtr=&/PowerupRifle_Whip_Heavy_01}}
 {id=WavePlayer_6, properties {soundWaveAssetPtr=&PowerupRifle_Whip_Heavy_01}}
 {id=WavePlayer_7, properties {soundWaveAssetPtr=&PowerupRifle_Whip_Heavy_02}}
 {id=WavePlayer_8, properties {soundWaveAssetPtr=&/PowerupRifle_Whip_Heavy_03}}
 {id=WavePlayer_9, properties {soundWaveAssetPtr=&/PowerupRifle_Whip_Heavy_04}}
]
properties {
 bOverrideAttenuation=true
 maxDistance=2097152
 soundClassObject=&Fort_WeaponRanged_Projectile_Effects
 volumeMultiplier=0.8
 attenuationOverrides {bAttenuate=false, stereoSpread=0}
 concurrencySet [/Game/Sounds/Concurrency_Rules/Athena_BulletWhizBy_Conc]
}
